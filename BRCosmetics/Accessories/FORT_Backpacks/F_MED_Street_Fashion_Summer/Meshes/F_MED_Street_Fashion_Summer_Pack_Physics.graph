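# Physics asset for the F_MED_Street_Fashion_Summer backpack cosmetic, defining physics bodies for its pack and dangling gear.
# Capsules cover the bag, straps, bandoliers, cans, grenade and magazine, with a box for the dagger and a small sphere on the pack bone.
# Most pieces constrain directly to the pack, while both straps chain off the bag.
# Tight swing and twist limits restrict each piece to slight movement.
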PhysicsAsset /BRCosmetics/Accessories/FORT_Backpacks/F_MED_Street_Fashion_Summer/Meshes/F_MED_Street_Fashion_Summer_Pack_Physics
bodies {
 dyn_bag [{shapeType=Capsule, length=11.6733, radius=1.5046, center=<0.9728 -0.1013 -0.4565>, rotation=<rot -2.2827 -89.9999 -1.8531>}]
 dyn_bandolier_01 [{shapeType=Capsule, length=3.9059, radius=2.9022, center=<-1.4903 -0.7641 -2.8936>, rotation=<rot -0.0956 -3.9556 -94.0563>}]
 dyn_bandolier_02 [{shapeType=Capsule, length=2.7607, radius=3.3277, center=<3.2157 -0.7108 -1.9541>, rotation=<rot 0.1891 -4.9746 -88.1765>}]
 dyn_can_01 [{shapeType=Capsule, length=5.5762, radius=0.8898, center=<2.7849 0.0068 0.2035>, rotation=<rot 0.4607 -86.6652 -98.2088>}]
 dyn_can_02 [{shapeType=Capsule, length=5.2043, radius=0.796, center=<2.512 0.0009 0.222>, rotation=<rot 0 90.0908 -89.9486>}]
 dyn_can_03 [{shapeType=Capsule, length=5.2042, radius=0.7971, center=<2.5121 0.0067 0.0015>, rotation=<rot 0 90.1408 -89.941>}]
 dyn_dagger [{
   shapeType=Box
   extentX=1.6155
   extentY=3.717
   extentZ=19.8482
   shapeName=dyn_dagger_box
   center=<6.0819 0.4076 -0.0726>
   rotation=<rot 0 -87.1752 -89.6771>
  }]
 dyn_grenade [{shapeType=Capsule, length=4.7419, radius=2.8518, center=<4.6145 0.2112 -0.0866>, rotation=<rot -0.895 89.0384 -95.1172>}]
 dyn_magazine [{shapeType=Capsule, length=6.5286, radius=1.5956, center=<2.5081 -0.0166 0.0394>, rotation=<rot 0 -89.7999 -89.4484>}]
 dyn_strap_01 [{shapeType=Capsule, length=5.7684, radius=1.3429, center=<0.3049 -0.3738 0.008>, rotation=<rot 1.6123 -90 -2.5618>}]
 dyn_strap_02 [{shapeType=Capsule, length=5.6681, radius=1.5334, center=<0.3321 0.6324 0.2855>, rotation=<rot -0.0043 -89.9999 -5.6921>}]
 pack [{shapeType=Sphere, radius=1.0887, center=<1.1817 -5.1107 -0.0432>}]
}
constraints [
 {bone1Name=dyn_bandolier_01, bone2Name=pack, swing1LimitDegrees=1, swing2LimitDegrees=8, twistLimitDegrees=1}
 {bone1Name=dyn_bandolier_02, bone2Name=pack, swing1LimitDegrees=1, swing2LimitDegrees=10, twistLimitDegrees=5}
 {bone1Name=dyn_bag, bone2Name=pack, swing1LimitDegrees=4.5, swing2LimitDegrees=1, twistLimitDegrees=1}
 {bone1Name=dyn_grenade, bone2Name=pack, swing1LimitDegrees=1, swing2LimitDegrees=7, twistLimitDegrees=1}
 {bone1Name=dyn_magazine, bone2Name=pack, swing1LimitDegrees=10, swing2LimitDegrees=7, twistLimitDegrees=1}
 {bone1Name=dyn_strap_01, bone2Name=dyn_bag, swing1LimitDegrees=6.5, swing2LimitDegrees=1, twistLimitDegrees=1}
 {bone1Name=dyn_can_02, bone2Name=pack, swing1LimitDegrees=1, swing2LimitDegrees=8, twistLimitDegrees=1}
 {bone1Name=dyn_strap_02, bone2Name=dyn_bag, swing1LimitDegrees=6.5, swing2LimitDegrees=1, twistLimitDegrees=1}
 {bone1Name=dyn_can_01, bone2Name=pack, swing1LimitDegrees=1, swing2LimitDegrees=5, twistLimitDegrees=1}
 {bone1Name=dyn_can_03, bone2Name=pack, swing1LimitDegrees=1, swing2LimitDegrees=5, twistLimitDegrees=1}
 {bone1Name=dyn_dagger, bone2Name=pack, swing1LimitDegrees=5, swing2LimitDegrees=5, twistLimitDegrees=1}
]
